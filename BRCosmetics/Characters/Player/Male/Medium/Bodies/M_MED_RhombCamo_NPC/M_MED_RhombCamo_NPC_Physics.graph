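# Physics asset for the M_MED_RhombCamo_NPC male body, covering a dynamic coat over pelvis and leg bodies.
# Side, front and back coat chains hang from small master bodies, with limited swing and mostly locked twist; lower front panels use boxes, the rest capsules.
# Front masters constrain to the thighs, while some side and back masters constrain to the pelvis.
# Extra angularly free constraints cross-link neighboring lower coat panels.
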PhysicsAsset /BRCosmetics/Characters/Player/Male/Medium/Bodies/M_MED_RhombCamo_NPC/M_MED_RhombCamo_NPC_Physics
bodies {
 calf_l [{shapeType=Capsule, length=30, radius=7, shapeName=calf_l_capsule, center=<-16.4517 -2.1846 0>, rotation=<rot 90 172.875 170.426>}]
 calf_r [{shapeType=Capsule, length=30, radius=7, shapeName=calf_l_capsule, center=<14.0093 3.081 -1.1195>, rotation=<rot 90 0 0>}]
 dyn_coat_01_l [{
   shapeType=Capsule
   length=9.3062
   radius=4
   shapeName=dyn_coat_l_01_capsule
   center=<5.5954 -0.4212 1.3583>
   rotation=<rot 85.1248 6.034 8.1269>
  }]
 dyn_coat_01_r [{
   shapeType=Capsule
   length=9.3062
   radius=4
   shapeName=dyn_coat_l_01_capsule
   center=<5.5954 -0.4212 1.3583>
   rotation=<rot 85.1248 6.034 8.1269>
  }]
 dyn_coat_03_l [{
   shapeType=Capsule
   length=15.4546
   radius=4
   shapeName=dyn_coat_l_03_capsule
   center=<5.9556 -0.1479 1.5356>
   rotation=<rot 64.4843 -89.9999 0>
  }]
 dyn_coat_03_r [{shapeType=Capsule, length=20, radius=3.5, shapeName=dyn_coat_l_02_capsule, center=<2.2589 0.0153 0.1351>, rotation=<rot 0 0.0001 66.3474>}]
 dyn_coat_04_l [{shapeType=Capsule, length=13, radius=4.25, shapeName=dyn_coat_l_03_capsule, center=<2.4849 0.0543 -0.564>, rotation=<rot 0 0.0001 66.3474>}]
 dyn_coat_04_r [{shapeType=Capsule, length=13, radius=4.25, shapeName=dyn_coat_l_03_capsule, center=<2.4849 0.0543 -0.564>, rotation=<rot 0 0.0001 66.3474>}]
 dyn_coat_back_01_l [{
   shapeType=Capsule
   length=8.3573
   radius=3.75
   shapeName=dyn_coat_bk_l_02_capsule
   center=<7.8009 -0.0218 1.0514>
   rotation=<rot 82.4343 -6.4495 -6.3099>
  }]
 dyn_coat_back_01_r [{
   shapeType=Capsule
   length=9.3062
   radius=3.75
   shapeName=dyn_coat_bk_r_01_capsule
   center=<7.7271 0.3982 2.0269>
   rotation=<rot -84.3491 -172.4936 172.5777>
  }]
 dyn_coat_back_03_l [{
   shapeType=Capsule
   length=8.3573
   radius=3.25
   shapeName=dyn_coat_bk_l_02_capsule
   center=<4.9902 -0.5611 -0.1711>
   rotation=<rot 83.4563 -6.44 -6.3193>
  }]
 dyn_coat_back_03_r [{
   shapeType=Capsule
   length=8.3573
   radius=3.25
   shapeName=dyn_coat_bk_r_02_capsule
   center=<4.9944 0.4854 0.2582>
   rotation=<rot -83.4555 -174.3897 174.4518>
  }]
 dyn_coat_back_04_l [{
   shapeType=Capsule
   length=10
   radius=3
   shapeName=dyn_coat_bk_l_03_capsule
   center=<4.5538 -0.7195 -1.6127>
   rotation=<rot 83.2326 -6.4421 -6.3173>
  }]
 dyn_coat_back_04_r [{
   shapeType=Capsule
   length=10
   radius=3
   shapeName=dyn_coat_bk_r_03_capsule
   center=<4.6105 0.8139 -1.2021>
   rotation=<rot -87.8786 -174.4107 174.4308>
  }]
 dyn_coat_front_01_l [{
   shapeType=Capsule
   length=8
   radius=3
   shapeName=dyn_cape_04_l_capsule
   center=<5.9257 0.8425 2.1405>
   rotation=<rot 86.66 -179.9994 -179.9994>
  }]
 dyn_coat_front_01_r [{
   shapeType=Capsule
   length=10
   radius=6
   shapeName=dyn_cape_04_l_capsule
   center=<7.4278 -0.2834 4.2291>
   rotation=<rot 85.4846 171.2096 173.1799>
  }]
 dyn_coat_front_03_l [{
   shapeType=Box
   extentX=3
   extentY=24
   extentZ=23
   shapeName=dyn_L_frontCoat_2_box
   center=<7.0533 0 1.9039>
   rotation=<rot 88.8094 -0.0001 -0.0001>
  }]
 dyn_coat_front_03_r [{
   shapeType=Box
   extentX=3
   extentY=23
   extentZ=25
   shapeName=dyn_L_frontCoat_2_box
   center=<7.2463 -0.713 1.3876>
   rotation=<rot 88.004 179.9999 179.9999>
  }]
 dyn_coat_front_04_l [{shapeType=Box, extentX=3, extentY=20, extentZ=15, shapeName=dyn_L_frontCoat_3_box, center=<5.725 0 1.2295>, rotation=<rot 85 0 0>}]
 dyn_coat_front_04_r [{
   shapeType=Box
   extentX=3
   extentY=22
   extentZ=15
   shapeName=dyn_L_frontCoat_3_box
   center=<4.3348 -0.713 1.746>
   rotation=<rot 88.3901 0.0002 0.0002>
  }]
 dyn_mstr_coat_01_l1 [{shapeType=Capsule, length=0.5, radius=0.5, shapeName=dyn_mstr_coat_01_l1_capsule}]
 dyn_mstr_coat_01_r1 [{shapeType=Capsule, length=0.5, radius=0.5, shapeName=dyn_mstr_coat_01_r1_capsule}]
 dyn_mstr_coat_back_01_l1 [{shapeType=Capsule, length=0.5, radius=0.5, shapeName=dyn_mstr_coat_01_r1_capsule}]
 dyn_mstr_coat_back_01_r1 [{shapeType=Capsule, length=0.5, radius=0.5, shapeName=dyn_mstr_coat_01_r1_capsule}]
 dyn_mstr_coat_front_01_l [{shapeType=Capsule, length=0.5, radius=0.5, shapeName=dyn_mstr_coat_front_01_l_capsule}]
 dyn_mstr_coat_front_01_r [{shapeType=Capsule, length=0.5, radius=0.5, shapeName=dyn_mstr_coat_front_01_r_capsule}]
 pelvis [{shapeType=Capsule, length=2, radius=2, shapeName=pelvis_capsule, center=<-0.0223 0 0>}]
 thigh_l [{shapeType=Capsule, length=40, radius=9, shapeName=thigh_r_capsule, center=<-19.5579 0.0949 1.3028>, rotation=<rot 90 0 0>}]
 thigh_r [{shapeType=Capsule, length=40, radius=9, shapeName=thigh_r_capsule, center=<16.3952 0.7217 -2.8827>, rotation=<rot 88.5244 -29.0232 -29.1941>}]
}
constraints [
 {
  bone1Name=dyn_coat_01_l
  bone2Name=dyn_mstr_coat_01_l1
  swing1LimitDegrees=40
  swing2LimitDegrees=30
  twistLimitDegrees=5
  twistMotion=Locked
 }
 {bone1Name=dyn_coat_03_l, bone2Name=dyn_coat_01_l, swing1LimitDegrees=25, swing2LimitDegrees=30, twistLimitDegrees=1, twistMotion=Locked}
 {bone1Name=dyn_coat_04_l, bone2Name=dyn_coat_03_l, swing1LimitDegrees=20, swing2LimitDegrees=30, twistLimitDegrees=1, twistMotion=Locked}
 {
  bone1Name=dyn_mstr_coat_01_r1
  bone2Name=pelvis
  swing1LimitDegrees=45
  swing1Motion=Free
  swing2LimitDegrees=45
  swing2Motion=Free
  twistLimitDegrees=45
  twistMotion=Free
 }
 {
  bone1Name=dyn_coat_01_r
  bone2Name=dyn_mstr_coat_01_r1
  swing1LimitDegrees=25
  swing2LimitDegrees=25
  twistLimitDegrees=1
  twistMotion=Locked
 }
 {bone1Name=dyn_coat_03_r, bone2Name=dyn_coat_01_r, swing1LimitDegrees=20, swing2LimitDegrees=45, twistLimitDegrees=10}
 {bone1Name=dyn_coat_04_r, bone2Name=dyn_coat_03_r, swing1LimitDegrees=15, swing2LimitDegrees=40, twistLimitDegrees=0, twistMotion=Locked}
 {bone1Name=dyn_mstr_coat_front_01_l, bone2Name=thigh_l, swing1LimitDegrees=20, swing2LimitDegrees=30, twistLimitDegrees=0}
 {
  bone1Name=dyn_coat_front_01_l
  bone2Name=dyn_mstr_coat_front_01_l
  swing1LimitDegrees=25
  swing2LimitDegrees=30
  twistLimitDegrees=1
  twistMotion=Locked
 }
 {
  bone1Name=dyn_coat_front_03_l
  bone2Name=dyn_coat_front_01_l
  swing1LimitDegrees=15
  swing2LimitDegrees=35
  twistLimitDegrees=1
  twistMotion=Locked
 }
 {
  bone1Name=dyn_coat_front_04_l
  bone2Name=dyn_coat_front_03_l
  swing1LimitDegrees=20
  swing2LimitDegrees=40
  twistLimitDegrees=1
  twistMotion=Locked
 }
 {bone1Name=dyn_mstr_coat_front_01_r, bone2Name=thigh_r, swing1LimitDegrees=30, swing2LimitDegrees=30, twistLimitDegrees=0, twistMotion=Locked}
 {
  bone1Name=dyn_coat_front_01_r
  bone2Name=dyn_mstr_coat_front_01_r
  swing1LimitDegrees=10
  swing2LimitDegrees=30
  twistLimitDegrees=1
  twistMotion=Locked
 }
 {
  bone1Name=dyn_coat_front_03_r
  bone2Name=dyn_coat_front_01_r
  swing1LimitDegrees=10
  swing2LimitDegrees=30
  twistLimitDegrees=1
  twistMotion=Locked
 }
 {
  bone1Name=dyn_coat_front_04_r
  bone2Name=dyn_coat_front_03_r
  swing1LimitDegrees=10
  swing2LimitDegrees=45
  twistLimitDegrees=1
  twistMotion=Locked
 }
 {
  bone1Name=thigh_l
  bone2Name=pelvis
  swing1LimitDegrees=45
  swing1Motion=Free
  swing2LimitDegrees=45
  swing2Motion=Free
  twistLimitDegrees=45
  twistMotion=Free
 }
 {
  bone1Name=thigh_r
  bone2Name=pelvis
  swing1LimitDegrees=45
  swing1Motion=Free
  swing2LimitDegrees=45
  swing2Motion=Free
  twistLimitDegrees=45
  twistMotion=Free
 }
 {
  bone1Name=calf_l
  bone2Name=thigh_l
  swing1LimitDegrees=45
  swing1Motion=Free
  swing2LimitDegrees=45
  swing2Motion=Free
  twistLimitDegrees=45
  twistMotion=Free
 }
 {
  bone1Name=calf_r
  bone2Name=thigh_r
  swing1LimitDegrees=45
  swing1Motion=Free
  swing2LimitDegrees=45
  swing2Motion=Free
  twistLimitDegrees=45
  twistMotion=Free
 }
 {
  bone1Name=dyn_mstr_coat_back_01_l1
  bone2Name=pelvis
  swing1LimitDegrees=45
  swing1Motion=Free
  swing2LimitDegrees=45
  swing2Motion=Free
  twistLimitDegrees=45
  twistMotion=Free
 }
 {
  bone1Name=dyn_coat_back_01_l
  bone2Name=dyn_mstr_coat_back_01_l1
  swing1LimitDegrees=30
  swing2LimitDegrees=35
  twistLimitDegrees=1
  twistMotion=Locked
 }
 {
  bone1Name=dyn_coat_back_03_l
  bone2Name=dyn_coat_back_01_l
  swing1LimitDegrees=10
  swing2LimitDegrees=35
  twistLimitDegrees=1
  twistMotion=Locked
 }
 {
  bone1Name=dyn_coat_back_04_l
  bone2Name=dyn_coat_back_03_l
  swing1LimitDegrees=15
  swing2LimitDegrees=30
  twistLimitDegrees=1
  twistMotion=Locked
 }
 {
  bone1Name=dyn_coat_back_01_r
  bone2Name=dyn_mstr_coat_back_01_r1
  swing1LimitDegrees=30
  swing2LimitDegrees=35
  twistLimitDegrees=1
  twistMotion=Locked
 }
 {
  bone1Name=dyn_coat_back_03_r
  bone2Name=dyn_coat_back_01_r
  swing1LimitDegrees=20
  swing2LimitDegrees=38
  twistLimitDegrees=1
  twistMotion=Locked
 }
 {
  bone1Name=dyn_coat_back_04_r
  bone2Name=dyn_coat_back_03_r
  swing1LimitDegrees=15
  swing2LimitDegrees=30
  twistLimitDegrees=1
  twistMotion=Locked
 }
 {
  bone1Name=dyn_coat_back_04_l
  bone2Name=dyn_coat_04_l
  swing1LimitDegrees=45
  swing1Motion=Free
  swing2LimitDegrees=45
  swing2Motion=Free
  twistLimitDegrees=45
  twistMotion=Free
 }
 {
  bone1Name=dyn_coat_back_04_r
  bone2Name=dyn_coat_back_04_l
  swing1LimitDegrees=45
  swing1Motion=Free
  swing2LimitDegrees=45
  swing2Motion=Free
  twistLimitDegrees=45
  twistMotion=Free
 }
 {
  bone1Name=dyn_coat_03_r
  bone2Name=dyn_coat_back_03_r
  swing1LimitDegrees=45
  swing1Motion=Free
  swing2LimitDegrees=45
  swing2Motion=Free
  twistLimitDegrees=45
  twistMotion=Free
 }
 {
  bone1Name=dyn_coat_04_l
  bone2Name=dyn_coat_back_04_l
  swing1LimitDegrees=45
  swing1Motion=Free
  swing2LimitDegrees=45
  swing2Motion=Free
  twistLimitDegrees=45
  twistMotion=Free
 }
 {
  bone1Name=dyn_coat_03_l
  bone2Name=dyn_coat_back_03_l
  swing1LimitDegrees=45
  swing1Motion=Free
  swing2LimitDegrees=45
  swing2Motion=Free
  twistLimitDegrees=45
  twistMotion=Free
 }
 {
  bone1Name=dyn_coat_03_l
  bone2Name=dyn_coat_front_03_l
  swing1LimitDegrees=45
  swing1Motion=Free
  swing2LimitDegrees=45
  swing2Motion=Free
  twistLimitDegrees=45
  twistMotion=Free
 }
]
