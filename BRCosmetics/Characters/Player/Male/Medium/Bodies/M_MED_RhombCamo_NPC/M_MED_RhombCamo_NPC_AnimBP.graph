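# Anim Blueprint for the RhombCamo NPC body part, copying the attached parent's pose and adding coat dynamics.
# Additive bone offsets adjust the coat, weighted by crouch events and a right-thigh-driven Jump Alpha (its crouch select is a no-op), before pelvis-space RigidBody simulation.
# Each update computes gravity override via a deprecated function and feeds head-bone axes and an offset center into visor-pattern material parameters.
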AnimBlueprint /BRCosmetics/Characters/Player/Male/Medium/Bodies/M_MED_RhombCamo_NPC/M_MED_RhombCamo_NPC_AnimBP
target_skeleton=/BRCosmetics/Characters/Player/Male/Medium/Bodies/M_MED_RhombCamo_NPC/Meshes/M_MED_RhombCamo_NPC_Skeleton
graphs [
 AnimGraph {nodes [
   n1 = "Transform (Modify) Bone - Bone: dyn_mstr_coat_01_r1"(
    ComponentPose=Local To Component(LocalPose=Copy Pose From Mesh() {bUseAttachedParent=true})
    Alpha=Get Not Crouching() -> (Not Crouching:double)
    Translation="-1.314,0.000,-0.000"
    Rotation="-11.872,-1.513,7.321"
    Scale="1,1,1"
   ) {
    rotationMode=BMM_Additive
    translationMode=BMM_Additive
    alphaBoolBlend {blendOption=Linear}
    alphaScaleBiasClamp {bInterpResult=true}
    boneToModify {boneName=dyn_mstr_coat_01_r1}
    rotation=<rot -11.872 -1.513 7.321>
    translation {x=-1.314}
   }
   n2 = "Transform (Modify) Bone - Bone: dyn_mstr_coat_front_01_l"(ComponentPose=@n1, Alpha=Get Not Crouching(), Translation="0,0,0", Rotation="0,0,0", Scale="1,1,1") {
    rotationMode=BMM_Additive
    translationMode=BMM_Additive
    alphaBoolBlend {blendOption=Linear}
    alphaScaleBiasClamp {bInterpResult=true}
    boneToModify {boneName=dyn_mstr_coat_front_01_l}
   }
   n3 = "Transform (Modify) Bone - Bone: dyn_mstr_coat_front_01_l"(
    ComponentPose=@n2
    Alpha=Get Crouching Alpha() -> (Crouching Alpha:double)
    Translation="3.012,6.765,6.765"
    Rotation="4.301,-0.998,14.932"
    Scale="1,1,1"
   ) {
    rotationMode=BMM_Additive
    translationMode=BMM_Additive
    alphaBoolBlend {blendOption=Linear}
    alphaScaleBiasClamp {bInterpResult=true}
    boneToModify {boneName=dyn_mstr_coat_front_01_l}
    rotation=<rot 4.301 -0.998 14.932>
    translation=<3.012 6.765 6.765>
   }
   n4 = "Transform (Modify) Bone - Bone: dyn_mstr_coat_front_01_r"(ComponentPose=@n3, Alpha=Get Crouching Alpha(), Translation="-1.962,0,0", Rotation="-12.743,9.770,14.825", Scale="1,1,1") {
    rotationMode=BMM_Additive
    translationMode=BMM_Additive
    alphaBoolBlend {blendOption=Linear}
    alphaScaleBiasClamp {bInterpResult=true}
    boneToModify {boneName=dyn_mstr_coat_front_01_r}
    rotation=<rot -12.743 9.77 14.825>
    translation {x=-1.962}
   }
   n5 = "Transform (Modify) Bone - Bone: dyn_mstr_coat_front_01_r"(ComponentPose=@n4, Alpha=Get Jump Alpha() -> (Jump Alpha:double), Translation="-1.63,0,8", Rotation="-29.818,-4.582,-12", Scale="1,1,1") {
    rotationMode=BMM_Additive
    translationMode=BMM_Additive
    alphaBoolBlend {blendOption=Linear}
    alphaScaleBiasClamp {bInterpResult=true}
    boneToModify {boneName=dyn_mstr_coat_front_01_r}
    rotation=<rot -29.818 -4.582 -12>
    translation {x=-1.63, z=8}
   }
   n6 = RigidBody!(ComponentPose=@n5, Alpha, OverrideWorldGravity:Vector=Get GravityOverride(), ExternalForce) -> (Pose) {
    bClampLinearTranslationLimitToRefPose=true
    bOverrideWorldGravity=true
    overridePhysicsAsset=&/M_MED_RhombCamo_NPC_Physics
    simulationSpace=BaseBoneSpace
    alphaBoolBlend {blendOption=Linear}
    baseBoneRef {boneName=pelvis}
    componentAppliedLinearAccClamp=<3000 3000 3000>
    componentLinearAccScale=<0.5 0.3 0.3>
    componentLinearVelScale=<0.5 0.3 0.3>
    overrideWorldGravity {z=-980}
    simSpaceSettings {
     maxAngularAcceleration=1000
     maxAngularVelocity=3
     maxLinearAcceleration=700
     maxLinearVelocity=500
     velocityScaleZ=0
     worldAlpha=1
     externalLinearDragV=<0.2 0.2>
    }
   }
   Output Pose(Result=Component To Local(ComponentPose=@n6)) {layerGroup=None}
  ]}
 EventGraph {nodes [
   Event BlueprintUpdateAnimation()
   > n1 = Sequence!(execute) -> (then_0, then_1, then_2:Exec)
   > n2 = Gravity Override State Function DEPRECATED(CharacterPartAnimInstance=Self-Reference(), GameTimeInSeconds=GetGameTimeInSeconds(), GravityOverrideParams=Get Gravity Override Params())
   > Set GravityOverride(GravityOverride=@n2)
   n3 = GetOwnerSkeletalMeshComponent()
   n4 = BreakRotator(InRot=BreakTransform(InTransform=GetDeltaTransformFromRefPose(self=@n3, BoneName=thigh_r, BaseName=pelvis)).Rotation)
   n5 = Get bIsCrouching()
   Set Jump Alpha(
    execute=@n1.then_1
    Jump Alpha:double=float * float(A=MapRangeClamped(Value=@n4.Yaw, InRangeA=0, InRangeB=-60), B=SelectFloat(A=1, B=1, bPickA=@n5)) {is=title K2Node_PromotableOperator}
   ) -> (Output_Get:double)
   n6 = Get Dynamic Material Instance(execute=@n1.then_2, SkeletalMeshComponent=GetPartSkeletalMeshComponent())
   n7 = BreakTransform(InTransform=GetBoneTransform!(
     self:Skinned Mesh Component Object Reference=GetOwnerSkeletalMeshComponent()
     InBoneName:Name=head
     TransformSpace:ERelativeTransformSpace Enum=RTS_World
    ) -> (ReturnValue:Transform))
   n8 = GetRightVector!(InRot:Rotator=@n7.Rotation) -> (ReturnValue:Vector)
   > SetVectorParameterValue(
    self=@n6
    ParameterName=VP_PlaneNormal
    Value="To LinearColor (Vector)"!(InVec:Vector=vector * vector(A=@n8, B:Vector="-1,-1,-1") -> (ReturnValue=~)) -> (ReturnValue:Linear Color Structure)
   )
   n9 = GetUpVector!(InRot:Rotator=@n7.Rotation) -> (ReturnValue:Vector)
   > SetVectorParameterValue(self=@n6, ParameterName=VP_XAxis, Value="To LinearColor (Vector)"(InVec=vector * vector(A=@n9, B="-1,-1,-1") -> (ReturnValue=~)))
   n10 = BreakVector(InVec=Get VisorPatternOffset() -> (VisorPatternOffset:Vector))
   n11 = vector * vector(A=@n8, B:double=@n10) -> (ReturnValue=~)
   n12 = vector * vector(A=@n9, B=@n10.Y) -> (ReturnValue=~)
   n13 = vector * vector(A=GetForwardVector!(InRot:Rotator=@n7.Rotation) -> (ReturnValue:Vector), B=@n10.Z) -> (ReturnValue=~)
   > SetVectorParameterValue(
    self=@n6
    ParameterName=VP_PlaneCenter
    Value="To LinearColor (Vector)"(InVec=vector + vector!(A=@n7, B=@n11, C:Vector=@n12, D:Vector=@n13) -> (ReturnValue))
   )
   Event OnBeginCrouching()
   > Set crouching alpha(crouching alpha:double=1) -> (Output_Get:double)
   > Set Not Crouching(Not Crouching:double=0) -> (Output_Get:double)
   Event OnEndCrouching()
   > Set crouching alpha(crouching alpha=0)
   > Set Not Crouching(Not Crouching=1)
  ]}
]
variables [
 Crouching Alpha {type=number}
 Gravity Override Params {type=GravityOverrideParamsStruct}
 GravityOverride {type=Vector}
 Jump Alpha {type=number}
 Not Crouching {type=number, default=1}
 VisorPatternOffset {type=Vector, default=<9.7 1 5>}
]
